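# Sound cue that plays when the LyricalFemale cosmetic glider opens.
# After a short delay, a mixer layers the open sound with an MX layer routed through a licensed content switcher.
# The switcher's first branch substitutes a looping silence test wave for the MX wave, which fills its other two branches.
# Shared glider open/close attenuation and concurrency govern playback, with the glider-open property sheet and the Fort_Gliders_Land sound class.
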SoundCue /BRCosmetics/Sounds/Gliders/LyricalFemale/GliderOpen_LyricalFemale_Cue
root=Delay_0
nodes [
 {id=Delay_0, children [Mixer_0], properties {delayMin=0.3}}
 {id=FortSoundNodeLicensedContentSwitcher_0, class=FortSoundNodeLicensedContentSwitcher, children [WavePlayer_2,WavePlayer_1,WavePlayer_1]}
 {id=Mixer_0, children [WavePlayer_0, FortSoundNodeLicensedContentSwitcher_0], properties {inputVolume [1.2,1.2]}}
 {id=WavePlayer_0, properties {soundWaveAssetPtr=&/Glider_LyricalFemale_Open}}
 {id=WavePlayer_1, properties {soundWaveAssetPtr=&/Glider_LyricalFemale_Open_MX}}
 {id=WavePlayer_2, properties {bLooping=true, soundWaveAssetPtr=/Game/Sounds/Fort_Test_Sounds/5_seconds_silence}}
]
properties {
 attenuationSettings=/Game/Sounds/Attenuation/Fort_Player_Glider_OpenClose_Athena_Att
 audioPropertiesSheet=/BRCosmetics/Sounds/Settings/PropertySheets/Glider/APS_Cosmetic_Glider_Open
 maxDistance=10000
 soundClassObject=&Fort_Gliders_Land
 volumeMultiplier=0.9
 concurrencySet [&Athena_GliderOpenClose_Conc]
}
